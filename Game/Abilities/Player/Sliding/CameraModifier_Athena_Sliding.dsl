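; Camera modifier Blueprint for the player's sliding ability in the Athena mode.
; A custom BlendOut event sets bBlendIn to false on the sliding camera modifier, flagging the slide camera effect to blend out.
; The blend behavior itself is not defined in this graph, which holds only that toggle.
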(event Custom|BlendOut
  (Variables|SlidingCameraModifier|SetbBlendIn false))
