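; Gameplay ability that plays a pre-slide camera shake as feedback when the player presses crouch and can proceed to sliding.
; On activation without authority, it caches the movement component and camera manager and binds to crouch input.
; Each press queries the external sliding provider, starts the shake, and binds targeting changes on a newly seen current weapon.
; Targeting stops the shake immediately, and ability end unbinds the cached weapon's targeting event.
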(event Ability|EventActivateAbility
  (bind _returnvalue (Fortnite|GetFortPlayerControllerfromActor (CallFunction|GetAvatarasFortPlayerPawn)))
  (bind _returnvalue_2 (CallFunction|GetAvatarasFortPlayerPawn))
  ;[0.3318 1 0.3148] Initialize values and bind to events
  (Ability|HasAuthority
    (:False
      (bind _asfort_movement_comp_character (Utilities|Casting|CastToFortMovementComp_Character (Variables|/Script/Engine.Character|GetCharacterMovement _returnvalue_2)))
      (Variables|Default|SetCachedMovementComponent _asfort_movement_comp_character)
      (Utilities|IsValid _returnvalue
        (:Is%20Valid
          (Variables|Default|SetCachedPlayerCameraManager (Variables|/Script/Engine.PlayerController|GetPlayerCameraManager _returnvalue))
          (Input|BindEventtoOnCrouchInputJustPressed _returnvalue_2 (delegate Custom|Crouch%20Input%20Pressed))
          ;[1 0.6311 0.2885] Stop camera shake if the user begins targeting.
          (bind _currentweapon (Variables|Default|GetCurrentWeapon)
            (:then
              (EventDispatchers|BindEventtoOnTargetingChanged _currentweapon (delegate Custom|OnTargetingChanged)))
            (:else)))))))

(event Custom|Crouch%20Input%20Pressed
  (bind _returnvalue (CallFunction|GetAvatarasFortPlayerPawn))
  (bind _returnvalue_2 (Weapon|GetCurrentWeapon _returnvalue))
  ;[0.3211 1 0.6558] Activate camera shake on valid crouch input
  (bind _cachedmovementcomponent (Variables|Default|GetCachedMovementComponent)
    (:then
      (if (Utilities|IsValidInterface (Class|FortMovementCompCharacter|GetExternalSlidingProvider _cachedmovementcomponent))
        (if (Class|FortMovementExternalSlidingProvider|CanPlayerProceedtoSliding (Class|FortMovementCompCharacter|GetExternalSlidingProvider _cachedmovementcomponent) false)
          (bind _cachedplayercameramanager (Variables|Default|GetCachedPlayerCameraManager)
            (:then
              (bind _returnvalue_3 (CameraShakes|StartCameraShake :self _cachedplayercameramanager :ShakeClass "/Game/Abilities/Player/Sliding/CamShake_Athena_Player_PreSlide.CamShake_Athena_Player_PreSlide_C" :Scale 1.0 :PlaySpace "CameraLocal"))
              (Variables|Default|SetCameraShake _returnvalue_3)
              ;[1 0.6311 0.2885] Stop camera shake if the user begins targeting.
              (if (!= (Variables|Default|GetCurrentWeapon) _returnvalue_2)
                (if (Utilities|IsValid _returnvalue_2 :!class /Script/Engine.KismetSystemLibrary)
                  (EventDispatchers|UnbindEventfromOnTargetingChanged _returnvalue_2 (delegate Custom|OnTargetingChanged)))
                (Variables|Default|SetCurrentWeapon _returnvalue_2)
                (bind _currentweapon (Variables|Default|GetCurrentWeapon)
                  (:then
                    (EventDispatchers|BindEventtoOnTargetingChanged _currentweapon (delegate Custom|OnTargetingChanged)))
                  (:else))))
            (:else)))))
    (:else)))

(event Custom|OnTargetingChanged
  (params (bIsTargeting :type Boolean))
  ;[1 0.6311 0.2885] Stop camera shake if the user begins targeting.
  (if bIsTargeting
    (bind _camerashake (Variables|Default|GetCameraShake)
      (:then
        (bind _cachedplayercameramanager (Variables|Default|GetCachedPlayerCameraManager)
          (:then
            (CameraShakes|StopCameraShake _cachedplayercameramanager _camerashake true))
          (:else)))
      (:else))))

(event Ability|EventFailedToActivatePassiveAbility
  (params (ActorInfo :type Gameplay%20Ability%20Actor%20Info%20Structure))
  (|CallParentFunction))

(event Ability|EventOnEndAbility
  (params (bWasCancelled :type Boolean))
  ;[1 0.6311 0.2885] Stop camera shake if the user begins targeting.
  (bind _currentweapon (Variables|Default|GetCurrentWeapon)
    (:then
      (EventDispatchers|UnbindEventfromOnTargetingChanged _currentweapon (delegate Custom|OnTargetingChanged)))
    (:else)))
