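; Gameplay ability for player sliding that drives slide cues, anim layer, land FX and camera effects.
; Activation saves sliding params and links the layer, re-linking after customization; non-server instances add a camera modifier and shake, stopped on ADS.
; AddMomentum, a re-armed timer, only handles speedlines and landing Niagara, re-arming only when not falling.
; Ending restores params and clears effects; ForceRemoveSlideGC force-clears stuck slides, while EndLevelGroundSlide is empty.
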(event Custom|EndLevelGroundSlide)

(event Ability|EventActivateAbilityFromEvent
  (params (EventData :type Gameplay%20Event%20Data%20Structure%20%28by%20ref%29))
  (bind _returnvalue (CallFunction|GetAvatarasFortPlayerPawn))
  (bind _returnvalue_2 (CallFunction|GetAvatarasFortPlayerPawn))
  (bind _returnvalue_3 (CallFunction|GetAvatarasFortPlayerPawn))
  (if (Utilities|IsValid _returnvalue_3 :!class /Script/Engine.KismetSystemLibrary)
    (Ability|SendGameplayEventtoActor _returnvalue_3 "(TagName=\"Athena.ContextualTutorial.EnterSlide.Complete\")")
    (Ability|AddGameplayCueToOwner (Variables|Default|GetCameraShakeGCNL))
    (Ability|AddGameplayCueToOwner (Variables|Default|GetPlayerGameplayCue))
    (Variables|Default|SetDefaultSlidingParams (Variables|/Script/FortniteGame.FortMovementComp_Character|GetSlidingControlParams (Utilities|Casting|CastToFortMovementComp_Character (Variables|/Script/Engine.Character|GetCharacterMovement (CallFunction|GetAvatarasFortPlayerPawn)))))
    (Utilities|FlowControl|Sequence
      (:then_0
        ;[1 1 1] Apply zero ground friction during initial Slide Dash
        (bind _returnvalue_4 (Utilities|Time|SetTimerbyFunctionName 0 "EndLevelGroundSlide" (ScalableFloat|GetValueatLevel (Variables|Default|GetDashDuration))))
        (Variables|Default|SetEndLevelGroundSlideTimerHandle _returnvalue_4))
      (:then_1
        (Rendering|Material|SetScalarParameterValue "/Game/Packages/Fortress_SharedMaterials/GlobalMaterialParameters/FortniteMaterialParameters.FortniteMaterialParameters" "IsPlayerSliding" 1.0 :!class /Script/Engine.KismetMaterialLibrary) ; This is meant to turn off the Z offset on landscape grass (based on camera loc)
        (Animation|MulticastLinkAnimLayer _returnvalue_2 "/Game/Characters/Player/Male/Male_Avg_Base/LayerAnimBP/SlidingLayerAnimBP.SlidingLayerAnimBP_C")
        (Customization|AssignOnFinishedCharacterCustomization _returnvalue_2 (delegate Custom|OnFinishedCharacterCustomization_Event_0))
        (bind _returnvalue_5 (Utilities|Time|SetTimerbyFunctionName 0 "AddMomentum" 0.05))
        (Variables|Default|SetAddMomentumTimerHandle _returnvalue_5)
        (if (not (Networking|IsServer))
          (bind _returnvalue_6 (Utilities|Time|SetTimerbyFunctionName 0 "CameraFX" 0.05))
          (Variables|Default|SetCameraFXDelay _returnvalue_6)))
      (:then_2
        (if (Networking|IsServer)
          (bind _asplayer_pawn_athena_generic (Utilities|Casting|CastToPlayerPawn_Athena_Generic _returnvalue)
            (:then
              (Class|PlayerPawnAthenaGeneric|ToggleLandFX _asplayer_pawn_athena_generic true))
            (:CastFailed
              (bind _asplayer_pawn_generic (Utilities|Casting|CastToPlayerPawn_Generic _returnvalue))
              (Class|PlayerPawnGeneric|ToggleLandFX _asplayer_pawn_generic true))))))))

(event Custom|CameraFX
  (bind _returnvalue (CallFunction|GetAvatarasFortPlayerPawn))
  (bind _returnvalue_2 (Fortnite|GetFortPlayerControllerfromActor _returnvalue))
  ;[1 1 1] Camera FX
  (Utilities|IsValid _returnvalue
    (:Is%20Valid
      (Utilities|IsValid _returnvalue_2
        (:Is%20Valid
          (bind _asfort_player_pawn_athena (Utilities|Casting|CastToFortPlayerPawnAthena _returnvalue)
            (:then
              (EventDispatchers|BindEventtoOnTargetingChanged (Weapon|GetCurrentWeapon _asfort_player_pawn_athena) (delegate Custom|OnTargetingChanged)))
            (:CastFailed
              _))
          (bind _returnvalue_3 (Fortnite|Camera|AddCameraModifierforOwningPlayerController (Variables|Default|GetCameraModifier) _returnvalue_2))
          (Variables|Default|SetSlideCameraModifier _returnvalue_3)
          (bind _returnvalue_4 (CameraShakes|StartCameraShake :self (Variables|/Script/Engine.PlayerController|GetPlayerCameraManager (Fortnite|GetFortPlayerControllerfromActor (CallFunction|GetAvatarasFortPlayerPawn))) :ShakeClass (Variables|Default|GetStartSlideCameraShake) :Scale 1.0 :PlaySpace "CameraLocal"))
          (Variables|Default|SetCameraShakeObjRef _returnvalue_4))))))

(event Custom|OnFinishedCharacterCustomization_Event_0
  (params (Pawn :type Fort%20Player%20Pawn%20Object%20Reference))
  (bind _returnvalue (CallFunction|GetAvatarasFortPlayerPawn))
  ;[1 1 1] Push back the anim layer after it has been removed from the customization reset
  (if (Utilities|IsValid _returnvalue :!class /Script/Engine.KismetSystemLibrary)
    (Animation|MulticastLinkAnimLayer _returnvalue "/Game/Characters/Player/Male/Male_Avg_Base/LayerAnimBP/SlidingLayerAnimBP.SlidingLayerAnimBP_C")))

(event Custom|OnTargetingChanged
  (params (bIsTargeting :type Boolean))
  (bind _camerashakeobjref (Variables|Default|GetCameraShakeObjRef))
  ;[1 1 1] Cancel the Camera Shake if we start to ADS
  (if bIsTargeting
    (if (Utilities|IsValid _camerashakeobjref :!class /Script/Engine.KismetSystemLibrary)
      (CameraShakes|StopCameraShake (Variables|/Script/Engine.PlayerController|GetPlayerCameraManager (Fortnite|GetFortPlayerControllerfromActor (CallFunction|GetAvatarasFortPlayerPawn))) _camerashakeobjref true)
      (Variables|Default|SetCameraShakeObjRef 0))))

(event Custom|ForceRemoveSlideGC
  (bind _returnvalue (CallFunction|GetAvatarasFortPlayerPawn))
  ;[1 0 0] Work around cheat for slide getting stuck on
  (CallFunction|EndAbilityServerSide)
  (CallFunction|EndAbilityClientSide)
  (Game|Player|RemoveCameraModifier (Variables|/Script/Engine.PlayerController|GetPlayerCameraManager (Fortnite|GetFortPlayerControllerfromActor (CallFunction|GetAvatarasFortPlayerPawn))) (Variables|Default|GetSlideCameraModifier))
  (Ability|RemoveGameplayCueFromOwner (Variables|Default|GetPlayerGameplayCue))
  (Ability|RemoveGameplayCueFromOwner (Variables|Default|GetCameraShakeGCNL))
  (Ability|RemoveGameplayCueFromOwner (Variables|Default|GetSpeedlinesCue))
  ;[1 1 1] End Ability
  (Utilities|Time|ClearandInvalidateTimerbyHandle (Variables|Default|GetCameraFXDelay))
  (Utilities|Time|ClearandInvalidateTimerbyHandle (Variables|Default|GetEndLevelGroundSlideTimerHandle))
  (Utilities|Time|ClearandInvalidateTimerbyHandle (Variables|Default|GetAddMomentumTimerHandle))
  ;;[1 1 1] Remove GC's
  (Ability|RemoveGameplayCueFromOwner (Variables|Default|GetPlayerGameplayCue))
  (Ability|RemoveGameplayCueFromOwner (Variables|Default|GetCameraShakeGCNL))
  (Utilities|IsValid _returnvalue
    (:Is%20Valid
      (Variables|/Script/FortniteGame.FortMovementComp_Character|SetSlidingControlParams (Variables|Default|GetDefaultSlidingParams) (Utilities|Casting|CastToFortMovementComp_Character (Variables|/Script/Engine.Character|GetCharacterMovement _returnvalue)))
      (Animation|MulticastUnlinkAnimLayer _returnvalue "/Game/Characters/Player/Male/Male_Avg_Base/LayerAnimBP/SlidingLayerAnimBP.SlidingLayerAnimBP_C")
      (Customization|UnbindEventfromOnFinishedCharacterCustomization _returnvalue (delegate Custom|OnFinishedCharacterCustomization_Event_0))
      (bind _asfort_player_pawn_athena (Utilities|Casting|CastToFortPlayerPawnAthena _returnvalue))
      (if (Utilities|IsValid (Weapon|GetCurrentWeapon _asfort_player_pawn_athena) :!class /Script/Engine.KismetSystemLibrary)
        (EventDispatchers|UnbindEventfromOnTargetingChanged (Weapon|GetCurrentWeapon _asfort_player_pawn_athena) (delegate Custom|OnTargetingChanged))))))

(event Custom|AddMomentum
  (bind _returnvalue (Math|Vector|VectorLength (Pawn|Components|CharacterMovement|GetLastUpdateVelocity (Variables|/Script/Engine.Character|GetCharacterMovement (CallFunction|GetAvatarasFortPlayerPawn)))))
  (bind _returnvalue_2 (CallFunction|GetAvatarasFortPlayerPawn))
  (bind _charactermovement (Variables|/Script/Engine.Character|GetCharacterMovement _returnvalue_2))
  (bind _returnvalue_3 (Pawn|Components|CharacterMovement|GetLastUpdateVelocity _charactermovement))
  (bind _speedlinescue (Variables|Default|GetSpeedlinesCue))
  ;[1 0.2592 0.2393] Old tick that checks velocity, and applies speedlines GC if above a certain speed. Should either be nativized or moved into a GC.
  (if (Utilities|IsValid _returnvalue_2 :!class /Script/Engine.KismetSystemLibrary)
    (if (not (AI|Components|NavMovement|IsFalling (Variables|/Script/Engine.Character|GetCharacterMovement _returnvalue_2)))
      (Utilities|FlowControl|Sequence
        (:then_0
          ;[1 1 1] If velocity is above 750, display speedlines. If it falls below 725, remove them. The lower threshold was added for FORT-571724 so the gameplay cue is not almost instantly removed. This was casuing abnormal behavior in the gameplay cue for simulated proxies.
          (if (> _returnvalue 7.5e+02)
            _
            (elif (< _returnvalue 725.0)
              (Ability|RemoveGameplayCueFromOwner _speedlinescue)))
          (Utilities|FlowControl|DoOnce false
            (:Completed
              (if (not (Camera|IsUsingFirstPersonCamera (Fortnite|GetFortPlayerControllerfromActor (CallFunction|GetAvatarasFortPlayerPawn))))
                (Ability|AddGameplayCueToOwner _speedlinescue)))))
        (:then_1
          (bind _asfort_movement_comp_character (Utilities|Casting|CastToFortMovementComp_Character (Variables|/Script/Engine.Character|GetCharacterMovement (CallFunction|GetAvatarasFortPlayerPawn))))
          (Utilities|FlowControl|Sequence
            (:then_0
              (if (and (!= (Variables|/Script/FortniteGame.FortMovementComp_Character|GetbInFallingGracePeriod _asfort_movement_comp_character) true) (Variables|Default|GetPrev_FallingGracePeriodState))
                ;[0.2777 1 0 0.328 12] landed event
                (Niagara|SpawnSystematLocation "/Game/Effects/Fort_Effects/Gameplay/Sliding/NS_Sliding_Landing_And_Continue.NS_Sliding_Landing_And_Continue" (.ImpactPoint (Collision|BreakHitResult (Utilities|Struct|BreakFindFloorResult (Variables|/Script/Engine.CharacterMovementComponent|GetCurrentFloor _asfort_movement_comp_character)))) (Math|Rotator|MakeRotfromZ (.ImpactNormal (Collision|BreakHitResult (Utilities|Struct|BreakFindFloorResult (Variables|/Script/Engine.CharacterMovementComponent|GetCurrentFloor _asfort_movement_comp_character))))))))
            (:then_1
              (Variables|Default|SetPrev_FallingGracePeriodState (Variables|/Script/FortniteGame.FortMovementComp_Character|GetbInFallingGracePeriod _asfort_movement_comp_character)))))
        (:then_2
          ;[1 1 1] Reset timer instead of enabling looping to more accurately simulate the behavior of a tick. Looping timers have a chance to execute multiple times in a frame, whereas resetting the timer will result in one execution.
          (bind _returnvalue_4 (Utilities|Time|SetTimerbyFunctionName 0 "AddMomentum" 0.05))
          (Variables|Default|SetAddMomentumTimerHandle _returnvalue_4))))))

(event Ability|EventOnEndAbility
  (params (bWasCancelled :type Boolean))
  (bind _returnvalue (CallFunction|GetAvatarasFortPlayerPawn))
  ;[1 1 1] End Ability
  (if (Networking|IsServer)
    (CallFunction|EndAbilityServerSide)
    (else
      (CallFunction|EndAbilityClientSide)))
  (Utilities|Time|ClearandInvalidateTimerbyHandle (Variables|Default|GetCameraFXDelay))
  (Utilities|Time|ClearandInvalidateTimerbyHandle (Variables|Default|GetEndLevelGroundSlideTimerHandle))
  (Utilities|Time|ClearandInvalidateTimerbyHandle (Variables|Default|GetAddMomentumTimerHandle))
  ;;[1 1 1] Remove GC's
  (Ability|RemoveGameplayCueFromOwner (Variables|Default|GetPlayerGameplayCue))
  (Ability|RemoveGameplayCueFromOwner (Variables|Default|GetCameraShakeGCNL))
  (Utilities|IsValid _returnvalue
    (:Is%20Valid
      (Variables|/Script/FortniteGame.FortMovementComp_Character|SetSlidingControlParams (Variables|Default|GetDefaultSlidingParams) (Utilities|Casting|CastToFortMovementComp_Character (Variables|/Script/Engine.Character|GetCharacterMovement _returnvalue)))
      (Animation|MulticastUnlinkAnimLayer _returnvalue "/Game/Characters/Player/Male/Male_Avg_Base/LayerAnimBP/SlidingLayerAnimBP.SlidingLayerAnimBP_C")
      (Customization|UnbindEventfromOnFinishedCharacterCustomization _returnvalue (delegate Custom|OnFinishedCharacterCustomization_Event_0))
      (bind _asfort_player_pawn_athena (Utilities|Casting|CastToFortPlayerPawnAthena _returnvalue))
      (if (Utilities|IsValid (Weapon|GetCurrentWeapon _asfort_player_pawn_athena) :!class /Script/Engine.KismetSystemLibrary)
        (EventDispatchers|UnbindEventfromOnTargetingChanged (Weapon|GetCurrentWeapon _asfort_player_pawn_athena) (delegate Custom|OnTargetingChanged))))))

(fn EndAbilityClientSide
  (bind _slidecameramodifier (Variables|Default|GetSlideCameraModifier))
  (Rendering|Material|SetScalarParameterValue "/Game/Packages/Fortress_SharedMaterials/GlobalMaterialParameters/FortniteMaterialParameters.FortniteMaterialParameters" "IsPlayerSliding" :!class /Script/Engine.KismetMaterialLibrary) ; This is meant to turn on the Z offset on landscape grass (based on camera loc)
  (Utilities|IsValid _slidecameramodifier
    (:Is%20Valid
      (bind _ascamera_modifier_athena_sliding (Utilities|Casting|CastToCameraModifier_Athena_Sliding _slidecameramodifier))
      (Class|CameraModifierAthenaSliding|BlendOut _ascamera_modifier_athena_sliding))))

(fn EndAbilityServerSide
  (bind _returnvalue (CallFunction|GetAvatarasFortPlayerPawn))
  (if (Utilities|IsValid _returnvalue :!class /Script/Engine.KismetSystemLibrary)
    (bind _asplayer_pawn_athena_generic (Utilities|Casting|CastToPlayerPawn_Athena_Generic _returnvalue)
      (:then
        (Class|PlayerPawnAthenaGeneric|ToggleLandFX _asplayer_pawn_athena_generic false))
      (:CastFailed
        (bind _asplayer_pawn_generic (Utilities|Casting|CastToPlayerPawn_Generic _returnvalue))
        (Class|PlayerPawnGeneric|ToggleLandFX _asplayer_pawn_generic false)))))
